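# Niagara trail effect for the Glider_Stars cosmetic glider, active when User.bIsFullyDeployed is set.
# Stars spawns hue-shifted fading sprites between skeletal socket pairs, with size and offset switched by ForwardDot.
# Scanline_Trail draws two hue-shifted ribbons from the skeleton, facing opposite ways at spawn and stretching along local Y with age.
# An FN_Integrator value scales star rate and trail speeds, front-end modifiers retune values, and the AnimTrail emitter is disabled.
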NiagaraSystem /BRCosmetics/Gadgets/Assets/VinderTech_GliderChute/Glider_Stars/VFX/NS_GliderStars
dependencies {
 usedDataInterfaces [&NiagaraDataInterfaceCurve]
 usedDynamicInputs [&ClampFloat]
 usedModules [&SystemState, &FN_Integrator]
}
emitters {
 FN_AnimTrail001 {
  bEnabled=false
  emitterUpdateScript [
   &EmitterState(Life Cycle Mode=::NewEnumerator1, Loop Duration=1, UseLoopDelay=false)
   &FN_SpawnAnimTrail_v2(
    AnimTrailWidthMode=::FromFirst
    Number of Anim Trails:NiagaraInt32=&MakeCustomIntFromBool(Bool=$User.bIsFullyDeployed, True INT=2, False INT=0)
    Spawn Mode:ENiagaraSkeletalSpawning=::NewEnumerator0
   )
  ]
  particleSpawnScript [
   /Game/Effects/Niagara/Modules/AnimTrail/FN_InitializeAnimTrail/FN_InitializeAnimTrailBoneIndecies(First Bone or Socket Index=1)
   &FN_InitializeAnimTrailBoneIndecies(AnimTrail Selection=1, First Bone or Socket Index=2, Second Bone or Socket Index=3) {module=FN_InitializeAnimTrailBoneIndecies001}
   /Game/Effects/Niagara/Modules/AnimTrail/FN_InitializeAnimTrail/FN_InitializeAnimTrail_v3/FN_InitializeAnimTrail(Lifetime=1.25)
   &AddVelocity()
  ]
  particleUpdateScript [
   &AddVelocity(Velocity=&MakeVector(X=0, Y=&Lerp_Float(A=3, B=12, Alpha=$System.FN_Integrator.Value), Z=0)) {module=AddVelocity001}
   &SolveForcesAndVelocity(Clamp Velocity=false, Limit Acceleration=false)
   /Game/Effects/Niagara/Modules/AnimTrail/FN_UpdateAnimTrail/FN_UpdateAnimTrail_v3/FN_UpdateAnimTrail(Locked Particle RibbonWidth=1)
   &FN_RibbonMaterial(
    Opacity Pan U Speed:NiagaraFloat=&MakeCustomFloatFromBool(Bool=HlslExpression() {hlslExpression="User.ForwardDot == 1"}, True Float=1.9, False Float=1.5)
    Opacity Power=0.1
   )
   /Game/Effects/Niagara/Modules/Color/FN_ColorHueShift(Color:LinearColor=&MultiplyLinearColor(
     Linear Color=<rgba 1.5 0 3 1>
     Scale Factor=&MakeFloat4FromVectorAndFloat(
      "Vector (RGB)"=<1 1 1>
      "Float (Alpha)"=&FloatFromCurve(
       FloatCurve=DataInterface() {propertyValues {Curve {keys [
           {interpMode=RCIM_Cubic, tangentMode=RCTM_Break, time=0, value=1}
           {arriveTangent=-1.8212, interpMode=RCIM_Cubic, tangentMode=RCTM_Break, time=1, value=0}
          ]}}}
       Scale Curve=&MakeCustomFloatFromBool(Bool=$User.bIsFullyDeployed, True Float=1, False Float=0)
      )
     )
    ))
   /Game/Effects/Niagara/Modules/Materials/FN_RibbonMaterial/FN_RibbonMaterial_v2(Opacity Pan U Speed=0.45, Opacity Power=0.1, Opacity Scale U=0.7) {enabled=false}
  ]
  rendererClasses [&NiagaraRibbonRendererProperties]
  renderers [{rendererClass=&NiagaraRibbonRendererProperties}]
 }
 Scanline_Trail {
  emitterUpdateScript [
   &EmitterState(Life Cycle Mode=::NewEnumerator1, Loop Duration=1, UseLoopDelay=false)
   &FN_SpawnRibbon_v2(
    Number of Ribbons=2
    SpawnRate:NiagaraFloat=&MakeCustomFloatFromBool(
     Bool=$User.bIsFullyDeployed
     True Float=&FN_FrontEndFloatModifier(Gameplay=30, Front End=100)
     False Float=0
    )
   )
  ]
  particleSpawnScript [
   &FN_InitializeRibbon_v3(
    Color:LinearColor=<rgba 1 0.15 1 1>
    Enable Skeletal Position=true
    Facing Axis:ENiagaraOrientationAxis=::ZAxis
    Interpolate Bone Sampling=false
    Lifetime=1.5
    Ribbon Width:NiagaraFloat=&FN_FrontEndFloatModifier(Gameplay=125, Front End=92)
   )
   /SetVariables(Particles.RibbonFacing=&RotateVectorByQuaternion(
     Vector To Rotate=&MakeVector(
      X=0
      Y=0
      Z=&MakeCustomFloatFromBool(Bool=HlslExpression() {hlslExpression="Particles.SpawnGroup == 0"}, True Float=1, False Float=-1)
     )
     Quaternion=$Particles.SampledRotation
    )) {isSetParameters=true}
   /SetVariables(Particles.Position=&Add_Vector(
     A=Linked() {linkedVariable {name=Particles.Position, type=NiagaraPosition}}
     B=&FN_FrontEndVector3Modifier(Gameplay=<0 0 0>, Front End=<2 -2 0>)
    )) {isSetParameters=true}
   &AddVelocity(Velocity=<0 200 0>) {enabled=false}
  ]
  particleUpdateScript [
   &ParticleState()
   /SetVariables(Particles.Position=&Add_Vector(
     A=Linked() {linkedVariable {name=Particles.Initial.Position, type=NiagaraPosition}}
     B=&Multiply_VectorByFloat(
      Vector=&RotateVectorByQuaternion(
       Vector To Rotate=&MakeVector(
        X=0
        Y=&FN_FrontEndFloatModifier(Gameplay=&Lerp_Float(A=150, B=350, Alpha=$System.FN_Integrator.Value), Front End=350)
        Z=0
       )
       Quaternion=$Particles.SampledRotation
      )
      Float=$Particles.NormalizedAge
     )
    )) {isSetParameters=true}
   /Game/Effects/Niagara/Modules/Contrails/FN_UpdateRibbon/FN_UpdateRibbon_v3(Enable Skeletal Position=true)
   &SolveForcesAndVelocity(Clamp Velocity=false, Limit Acceleration=false)
   /SetVariables(Particles.RibbonFacing=&MakeCustomVectorFromBool(
     Bool=$Particles.Locked
     True Vector=&RotateVectorByQuaternion(Vector To Rotate=<0 0 -1>, Quaternion=$Particles.FN_UpdateRibbon_v3.SampledRotation)
     False Vector=$Particles.RibbonFacing
    )) {isSetParameters=true}
   &FN_RibbonMaterial(
    Opacity Pan U Offset:NiagaraFloat=&MakeCustomFloatFromBool(Bool=HlslExpression() {hlslExpression="Particles.SpawnGroup == 0 "}, True Float=0.15, False Float=0)
    Opacity Pan U Speed:NiagaraFloat=&MakeCustomFloatFromBool(Bool=HlslExpression() {hlslExpression="Particles.SpawnGroup == 0 "}, True Float=1.5, False Float=1.6)
   )
   &FN_ColorHueShift(Color:LinearColor=&MultiplyLinearColor(
     Linear Color=<rgba 1 0 2 1>
     Scale Factor=&MakeFloat4FromVectorAndFloat(
      "Vector (RGB)"=<1 1 1>
      "Float (Alpha)"=&FloatFromCurve(FloatCurve=DataInterface() {propertyValues {Curve {keys [
           {interpMode=RCIM_Cubic, leaveTangent=2.6802, tangentMode=RCTM_Break, time=0.0071, value=0.35}
           {interpMode=RCIM_Cubic, leaveTangent=-0.0682, tangentMode=RCTM_Break, time=0.5, value=1}
           {arriveTangent=0.0489, interpMode=RCIM_Cubic, tangentMode=RCTM_Break, time=1, value=0}
          ]}}})
     )
    ))
   /SetVariables(Particles.Scanline Color:LinearColor=$Particles.Color) {isSetParameters=true}
   /SetVariables(Particles.ScanlineRamp:LinearColor=&MakeLinearColorFromVectorAndFloat(
     "Vector (RGB)"=&Multiply_Vector(
      A=&MakeVectorFromLinearColorRGB(Color=$Particles.Color)
      B=&VectorFromFloat(Value=1)
     )
     "Float (Alpha)"=0.1
    )) {isSetParameters=true}
  ]
  rendererClasses [&NiagaraRibbonRendererProperties]
  renderers [
   {rendererClass=&NiagaraRibbonRendererProperties}
   {rendererClass=&NiagaraRibbonRendererProperties, rendererIndex=1}
  ]
 }
 Stars {
  emitterUpdateScript [
   &EmitterState(Life Cycle Mode=::NewEnumerator1, Loop Delay=3, Loop Duration=1, UseLoopDelay=false)
   &SpawnRate(
    Spawn Probability=&MakeCustomFloatFromBool(Bool=$User.bIsFullyDeployed, True Float=1, False Float=0)
    SpawnRate=&Lerp_Float(A=10, B=60, Alpha=$System.FN_Integrator.Value)
    Use Spawn Probability=true
   )
  ]
  particleSpawnScript [
   &InitializeParticle(
    Color Mode=::NewEnumerator0
    Lifetime=0.8
    Sprite Size Mode=::NewEnumerator1
    Uniform Sprite Size Max=&MakeCustomFloatFromBool(Bool=HlslExpression() {hlslExpression="User.ForwardDot > .5"}, True Float=35, False Float=21)
    Uniform Sprite Size Min=18
    UsePositionOffset=false
    Write Material Random=true
   )
   &FN_SetSpawnGroup(Number of Spawn Groups=2)
   &SkeletalMeshLocation(
    Apply To Particle Position=false
    Bone / Socket Index=HlslExpression() {hlslExpression=Particles.SpawnGroup * 2}
    Bone Sampling Mode=::NewEnumerator4
    Inherit Velocity=false
    MaskSpawnGroup=false
    Sampled Position Offset=<0 14199.8291 0>
    Sampled Position Offset Coordinate Space:ENiagaraMeshTransforms=::NewEnumerator3
    Skeletal Mesh=$User.SkeletalMesh
   )
   &SkeletalMeshLocation(
    Apply To Particle Position=false
    Bone / Socket Index=HlslExpression() {hlslExpression="(Particles.SpawnGroup * 2) + 1"}
    Bone Sampling Mode=::NewEnumerator4
    Inherit Velocity=false
    MaskSpawnGroup=false
    Sampled Position Offset=<0 63.0028 0>
    Sampled Position Offset Coordinate Space:ENiagaraMeshTransforms=::NewEnumerator3
    Skeletal Mesh=$User.SkeletalMesh
   ) {module=SkeletalMeshLocation001}
   /SetVariables(Particles.Position=&Lerp_Vector(
     A=&Add_Vector(
      A=$Output.SkeletalMeshLocation.SampledPosition
      B=&MakeVector(
       X=0
       Y=&MakeCustomFloatFromBool(Bool=HlslExpression() {hlslExpression="User.ForwardDot > 0.5"}, True Float=10, False Float=30)
       Z=0
      )
     )
     B=&Add_Vector(
      A=$Output.SkeletalMeshLocation001.SampledPosition
      B=&MakeVector(
       X=0
       Y=&MakeCustomFloatFromBool(Bool=HlslExpression() {hlslExpression="User.ForwardDot > 0.5"}, True Float=10, False Float=30)
       Z=0
      )
     )
     Alpha=&RandomRangeFloat(Minimum=0.25, Maximum=1)
    )) {isSetParameters=true}
   &AddVelocity(Velocity=&MakeVector(
     X=&FN_FrontEndFloatModifier(Gameplay=0, Front End=-125)
     Y=&FN_FrontEndFloatModifier(
      Gameplay=&MakeCustomFloatFromBool(
       Bool=$User.bIsFullyDeployed
       True Float=&Lerp_Float(A=115, B=250, Alpha=$System.FN_Integrator.Value)
       False Float=0
      )
      Front End=115
     )
     Z=-50
    ))
  ]
  particleUpdateScript [
   &ParticleState()
   &SolveForcesAndVelocity(Clamp Velocity=false, Limit Acceleration=false)
   &FN_ColorHueShift(Color:LinearColor=&MultiplyLinearColor(
     Linear Color=<rgba 1 0 2 1>
     Scale Factor=&MakeFloat4FromVectorAndFloat(
      "Vector (RGB)"=<1 1 1>
      "Float (Alpha)"=&FloatFromCurve(
       FloatCurve=DataInterface() {propertyValues {Curve {keys [
           {interpMode=RCIM_Cubic, tangentMode=RCTM_Break, time=0, value=1}
           {arriveTangent=-1.8212, interpMode=RCIM_Cubic, tangentMode=RCTM_Break, time=1, value=0}
          ]}}}
       Scale Curve=&FloatFromCurve(FloatCurve=DataInterface() {propertyValues {Curve {keys [
            {interpMode=RCIM_Cubic, leaveTangent=2.2929, tangentMode=RCTM_Break, time=0, value=0}
            {arriveTangent=-0.0113, interpMode=RCIM_Cubic, tangentMode=RCTM_Break, time=1, value=1}
           ]}}})
      )
     )
    ))
  ]
  rendererClasses [&NiagaraSpriteRendererProperties]
  renderers [{rendererClass=&NiagaraSpriteRendererProperties}]
 }
}
summary {
 systemName=NS_GliderStars
 emitters [
  {emitterName=Stars, rendererClasses [&NiagaraSpriteRendererProperties]}
  {bEnabled=false, emitterName=FN_AnimTrail001, rendererClasses [&NiagaraRibbonRendererProperties]}
  {emitterName=Scanline_Trail, rendererClasses [&NiagaraRibbonRendererProperties]}
 ]
 userVariables [
  User.bIsFrontEnd {type=NiagaraBool}
  User.bisFrontEndPreview {type=NiagaraBool}
  User.bIsFullyDeployed {type=NiagaraBool}
  User.ForwardDot {type=NiagaraFloat, defaultValue=0, description=""}
  User.Jet Color {type=LinearColor, description="", defaultValue {struct=LinearColor, value=<rgba 1 1 1 1>}}
  User.Jet_Alpha {type=NiagaraFloat, defaultValue=0, description=""}
  User.Ribbon Alpha {type=NiagaraFloat, defaultValue=1, description=""}
  User.Ribbon Color {type=LinearColor, description="", defaultValue {struct=LinearColor, value=<rgba 1 1 1 1>}}
  User.Ribbon Length {type=NiagaraFloat, defaultValue=0, description=""}
  User.Ribbon Width {type=NiagaraFloat, defaultValue=0, description=""}
  User.RibbonIDOffset {type=NiagaraInt32}
  User.RotationalMovement {type=NiagaraFloat, defaultValue=0, description=""}
  User.SkeletalMesh {
   type=NiagaraDataInterfaceSkeletalMesh
   defaultValue {
    struct=NiagaraExt_VariableValue_DataInterface
    value {dataInterface=&self:NiagaraDataInterfaceSkeletalMesh_1}
   }
  }
 ]
}
